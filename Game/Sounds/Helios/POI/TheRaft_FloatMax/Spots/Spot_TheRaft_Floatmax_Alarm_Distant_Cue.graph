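# Sound cue for a distant alarm spot sound at The Raft FloatMax point of interest, classed as an outdoor ambient one-shot.
# A random node picks between two equally weighted jail alarm waves at reduced volume.
# A shared distant linear attenuation shapes falloff, and a concurrency override caps simultaneous voices at two.
# Distance-based sends feed default and distant reverb submixes, each with a custom level curve.
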SoundCue /Game/Sounds/Helios/POI/TheRaft_FloatMax/Spots/Spot_TheRaft_Floatmax_Alarm_Distant_Cue
root=Random_2
nodes [
 {id=Random_2, children [WavePlayer_15, WavePlayer_16], properties {weights [1,1]}}
 {id=WavePlayer_15, properties {soundWaveAssetPtr=&/Spot_TheRaft_Jail_Alarm_02}}
 {id=WavePlayer_16, properties {soundWaveAssetPtr=&/Spot_TheRaft_Jail_Alarm_01}}
]
properties {
 attenuationSettings=/Game/Sounds/Attenuation/Fort_TileAmbient_OneShot_Distant_Creak_Linear_Att
 bOverrideConcurrency=true
 maxDistance=8700
 soundClassObject=&Fort_Ambient_OneShots_Outdoors
 volumeMultiplier=0.501
 concurrencyOverrides {maxCount=2, voiceStealReleaseTime=2}
 soundSubmixSends [
  {
   maxSendDistance=8500
   minSendDistance=200
   minSendLevel=0.6
   soundSubmix=&Fortnite_Reverb_Default
   customSendLevelCurve {editorCurveData {keys [<at 0 0.25>,<at 1 1>]}}
  }
  {
   maxSendDistance=8500
   maxSendLevel=0.6
   minSendDistance=200
   minSendLevel=0.25
   soundSubmix=&Fortnite_Reverb_Distant
   customSendLevelCurve {editorCurveData {keys [
      <at 0 0>
      {arriveTangent=1, interpMode=RCIM_Cubic, leaveTangent=1, time=0.66, value=0.33}
      {arriveTangent=2.1868, interpMode=RCIM_Cubic, leaveTangent=2.1868, tangentMode=RCTM_User, time=1, value=1}
     ]}}
  }
 ]
}
