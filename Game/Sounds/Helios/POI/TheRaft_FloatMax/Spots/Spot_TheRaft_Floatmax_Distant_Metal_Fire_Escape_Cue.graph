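# Sound cue for a distant metal fire escape ambient spot at The Raft FloatMax POI, reusing Hollow Base cave recordings.
# A quality-level node sends two tiers to a random pick of six waves and two to a pick of three, each through a modulator.
# The cue plays quiet and pitched down in the outdoor ambient class, with overridden long-falloff attenuation and submix and bus sends disabled.
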SoundCue /Game/Sounds/Helios/POI/TheRaft_FloatMax/Spots/Spot_TheRaft_Floatmax_Distant_Metal_Fire_Escape_Cue
root=QualityLevel_0
nodes [
 {id=Modulator_0, children [Random_0]}
 {id=Modulator_1, children [Random_1]}
 {id=QualityLevel_0, children [Modulator_0,Modulator_0,Modulator_1,Modulator_1]}
 {id=Random_0, children [WavePlayer_0,WavePlayer_1,WavePlayer_2,WavePlayer_3,WavePlayer_4,WavePlayer_5], properties {weights [1,1,1,1,1,1]}}
 {id=Random_1, children [WavePlayer_6,WavePlayer_7,WavePlayer_8], properties {weights [1,1,1]}}
 {id=WavePlayer_0, properties {soundWaveAssetPtr=/Game/Sounds/Artemis/POI/Hollow_Base/Spots/Spot_Cave_Distant_Metal_Fire_Escape_1}}
 {id=WavePlayer_1, properties {soundWaveAssetPtr=/Game/Sounds/Artemis/POI/Hollow_Base/Spots/Spot_Cave_Distant_Metal_Fire_Escape_2}}
 {id=WavePlayer_2, properties {soundWaveAssetPtr=/Game/Sounds/Artemis/POI/Hollow_Base/Spots/Spot_Cave_Distant_Metal_Fire_Escape_3}}
 {id=WavePlayer_3, properties {soundWaveAssetPtr=/Game/Sounds/Artemis/POI/Hollow_Base/Spots/Spot_Cave_Distant_Metal_Fire_Escape_4}}
 {id=WavePlayer_4, properties {soundWaveAssetPtr=/Game/Sounds/Artemis/POI/Hollow_Base/Spots/Spot_Cave_Distant_Metal_Fire_Escape_5}}
 {id=WavePlayer_5, properties {soundWaveAssetPtr=/Game/Sounds/Artemis/POI/Hollow_Base/Spots/Spot_Cave_Distant_Metal_Fire_Escape_6}}
 {id=WavePlayer_6, properties {soundWaveAssetPtr=&Spot_Cave_Distant_Metal_Fire_Escape_1}}
 {id=WavePlayer_7, properties {soundWaveAssetPtr=&Spot_Cave_Distant_Metal_Fire_Escape_2}}
 {id=WavePlayer_8, properties {soundWaveAssetPtr=&Spot_Cave_Distant_Metal_Fire_Escape_3}}
]
properties {
 attenuationSettings=/Game/Sounds/Attenuation/Fort_TileAmbient_OneShot_Distant_Creak_Att
 bEnableBaseSubmix=false
 bEnableBusSends=false
 bEnableSubmixSends=false
 bOverrideAttenuation=true
 maxDistance=6900
 pitchMultiplier=0.8
 soundClassObject=&Fort_Ambient_Outdoors_Athena
 volumeMultiplier=0.2
 attenuationOverrides {falloffDistance=6500}
}
