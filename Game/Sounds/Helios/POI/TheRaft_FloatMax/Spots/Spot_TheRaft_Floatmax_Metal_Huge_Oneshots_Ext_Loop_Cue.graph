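# Looping outdoor ambient spot cue for The Raft (FloatMax) POI, playing occasional huge metal creaks reused from Mount Olympus statue sounds.
# A quality-level root routes its first two slots to a nine-creak random pool and its last two to a three-creak pool.
# Each branch loops a randomly delayed pick with pitch and volume modulation.
# Lowered cue pitch and volume, overridden attenuation and a concurrency cap shape playback.
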SoundCue /Game/Sounds/Helios/POI/TheRaft_FloatMax/Spots/Spot_TheRaft_Floatmax_Metal_Huge_Oneshots_Ext_Loop_Cue
root=QualityLevel_0
nodes [
 {id=Delay_1, children [Random_0], properties {delayMax=42, delayMin=14}}
 {id=Delay_2, children [Random_1], properties {delayMax=42, delayMin=14}}
 {id=Looping_0, children [Delay_1]}
 {id=Looping_1, children [Delay_2]}
 {id=Modulator_0, children [Looping_1], properties {pitchMax=1.1, pitchMin=0.9, volumeMax=0.9, volumeMin=0.8}}
 {id=Modulator_1, children [Looping_0], properties {pitchMax=1.1, pitchMin=0.7, volumeMax=1.1, volumeMin=0.9}}
 {id=QualityLevel_0, children [Modulator_1,Modulator_1,Modulator_0,Modulator_0]}
 {
  id=Random_0
  children [WavePlayer_8, WavePlayer_9, WavePlayer_10, WavePlayer_11, WavePlayer_12, WavePlayer_13, WavePlayer_14, WavePlayer_15, WavePlayer_16]
  properties {weights [1,1,1,1,1,1,1,1,1]}
 }
 {id=Random_1, children [WavePlayer_17, WavePlayer_18, WavePlayer_19], properties {weights [1,1,1]}}
 {id=WavePlayer_10, properties {soundWaveAssetPtr=/Game/Sounds/Helios/POI/MountOlympus/Spots/Spot_Metal_Statue_Creak_03}}
 {id=WavePlayer_11, properties {soundWaveAssetPtr=/Game/Sounds/Helios/POI/MountOlympus/Spots/Spot_Metal_Statue_Creak_04}}
 {id=WavePlayer_12, properties {soundWaveAssetPtr=/Game/Sounds/Helios/POI/MountOlympus/Spots/Spot_Metal_Statue_Creak_05}}
 {id=WavePlayer_13, properties {soundWaveAssetPtr=/Game/Sounds/Helios/POI/MountOlympus/Spots/Spot_Metal_Statue_Creak_06}}
 {id=WavePlayer_14, properties {soundWaveAssetPtr=/Game/Sounds/Helios/POI/MountOlympus/Spots/Spot_Metal_Statue_Creak_07}}
 {id=WavePlayer_15, properties {soundWaveAssetPtr=/Game/Sounds/Helios/POI/MountOlympus/Spots/Spot_Metal_Statue_Creak_08}}
 {id=WavePlayer_16, properties {soundWaveAssetPtr=/Game/Sounds/Helios/POI/MountOlympus/Spots/Spot_Metal_Statue_Creak_09}}
 {id=WavePlayer_17, properties {soundWaveAssetPtr=&Spot_Metal_Statue_Creak_03}}
 {id=WavePlayer_18, properties {soundWaveAssetPtr=/Game/Sounds/Helios/POI/MountOlympus/Spots/Spot_Metal_Statue_Creak_02}}
 {id=WavePlayer_19, properties {soundWaveAssetPtr=/Game/Sounds/Helios/POI/MountOlympus/Spots/Spot_Metal_Statue_Creak_01}}
 {id=WavePlayer_8, properties {soundWaveAssetPtr=&Spot_Metal_Statue_Creak_01}}
 {id=WavePlayer_9, properties {soundWaveAssetPtr=&Spot_Metal_Statue_Creak_02}}
]
properties {
 attenuationSettings=/Game/Sounds/Attenuation/Fort_TileAmbient_OneShot_Att
 bEnableSubmixSends=false
 bOverrideAttenuation=true
 bOverrideConcurrency=true
 maxDistance=3000
 pitchMultiplier=0.8
 soundClassObject=&Fort_Ambient_OneShots_Outdoors
 volumeMultiplier=0.251
 attenuationOverrides {falloffDistance=2500, attenuationShapeExtents {x=500}}
 concurrencyOverrides {maxCount=3}
 soundSubmixSends [{maxSendLevel=8, minSendLevel=0, sendLevel=8, soundSubmix=&Fortnite_Reverb_Default}]
}
