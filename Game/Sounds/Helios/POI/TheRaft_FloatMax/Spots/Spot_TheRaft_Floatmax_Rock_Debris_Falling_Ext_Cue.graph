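# Sound cue for falling rock debris exterior spot ambience at The Raft FloatMax POI, reusing Hollow Base cave debris waves.
# A quality switch sends its first two slots through a modulator over a random pick of nine waves.
# The last two slots use an unmodulated random pick from a four-wave subset.
# It plays on an outdoor ambient sound class with shared ceiling-debris one-shot attenuation.
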SoundCue /Game/Sounds/Helios/POI/TheRaft_FloatMax/Spots/Spot_TheRaft_Floatmax_Rock_Debris_Falling_Ext_Cue
root=QualityLevel_0
nodes [
 {id=Modulator_0, children [Random_1]}
 {id=QualityLevel_0, children [Modulator_0,Modulator_0,Random_0,Random_0]}
 {id=Random_0, children [WavePlayer_0,WavePlayer_1,WavePlayer_2,WavePlayer_3], properties {weights [1,1,1,1]}}
 {
  id=Random_1
  children [WavePlayer_5, WavePlayer_6, WavePlayer_7, WavePlayer_8, WavePlayer_9, WavePlayer_10, WavePlayer_12, WavePlayer_13, WavePlayer_14]
  properties {weights [1,1,1,1,1,1,1,1,1]}
 }
 {id=WavePlayer_0, properties {soundWaveAssetPtr=/Game/Sounds/Artemis/POI/Hollow_Base/Spots/Spot_Cave_Rock_Debris_Falling_04}}
 {id=WavePlayer_1, properties {soundWaveAssetPtr=/Game/Sounds/Artemis/POI/Hollow_Base/Spots/Spot_Cave_Rock_Debris_Falling_08}}
 {id=WavePlayer_10, properties {soundWaveAssetPtr=/Game/Sounds/Artemis/POI/Hollow_Base/Spots/Spot_Cave_Rock_Debris_Falling_05}}
 {id=WavePlayer_12, properties {soundWaveAssetPtr=/Game/Sounds/Artemis/POI/Hollow_Base/Spots/Spot_Cave_Rock_Debris_Falling_06}}
 {id=WavePlayer_13, properties {soundWaveAssetPtr=/Game/Sounds/Artemis/POI/Hollow_Base/Spots/Spot_Cave_Rock_Debris_Falling_07}}
 {id=WavePlayer_14, properties {soundWaveAssetPtr=/Game/Sounds/Artemis/POI/Hollow_Base/Spots/Spot_Cave_Rock_Debris_Falling_09}}
 {id=WavePlayer_2, properties {soundWaveAssetPtr=/Game/Sounds/Artemis/POI/Hollow_Base/Spots/Spot_Cave_Rock_Debris_Falling_01}}
 {id=WavePlayer_3, properties {soundWaveAssetPtr=/Game/Sounds/Artemis/POI/Hollow_Base/Spots/Spot_Cave_Rock_Debris_Falling_02}}
 {id=WavePlayer_5, properties {soundWaveAssetPtr=&Spot_Cave_Rock_Debris_Falling_04}}
 {id=WavePlayer_6, properties {soundWaveAssetPtr=&Spot_Cave_Rock_Debris_Falling_08}}
 {id=WavePlayer_7, properties {soundWaveAssetPtr=&Spot_Cave_Rock_Debris_Falling_01}}
 {id=WavePlayer_8, properties {soundWaveAssetPtr=&Spot_Cave_Rock_Debris_Falling_02}}
 {id=WavePlayer_9, properties {soundWaveAssetPtr=/Game/Sounds/Artemis/POI/Hollow_Base/Spots/Spot_Cave_Rock_Debris_Falling_03}}
]
properties {
 attenuationSettings=/Game/Sounds/Attenuation/Fort_TileAmbient_OneShot_Ceiling_Debris_Att
 bEnableBusSends=false
 bEnableSubmixSends=false
 maxDistance=6750
 soundClassObject=&Fort_Ambient_Outdoors_Athena
 volumeMultiplier=0.752
 attenuationOverrides {bAttenuate=false}
}
