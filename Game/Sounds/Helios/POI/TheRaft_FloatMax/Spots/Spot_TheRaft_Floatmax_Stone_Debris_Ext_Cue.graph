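# Sound cue for one-shot exterior stone debris ambience at The Raft FloatMax POI, reusing Fortress interior large stone debris recordings.
# A Modulator varies output from a Quality Level switch, whose first two slots pick randomly among twelve waves and last two among only the first four.
# The cue lowers pitch and volume and uses outdoor ambient one-shot class with less-positional attenuation.
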SoundCue /Game/Sounds/Helios/POI/TheRaft_FloatMax/Spots/Spot_TheRaft_Floatmax_Stone_Debris_Ext_Cue
root=Modulator_0
nodes [
 {id=Modulator_0, children [QualityLevel_0]}
 {id=QualityLevel_0, children [Random_1,Random_1,Random_3,Random_3]}
 {
  id=Random_1
  children [
   WavePlayer_12
   WavePlayer_13
   WavePlayer_14
   WavePlayer_15
   WavePlayer_16
   WavePlayer_17
   WavePlayer_24
   WavePlayer_26
   WavePlayer_27
   WavePlayer_28
   WavePlayer_29
   WavePlayer_30
  ]
  properties {weights [1,1,1,1,1,1,1,1,1,1,1,1]}
 }
 {id=Random_3, children [WavePlayer_31, WavePlayer_32, WavePlayer_33, WavePlayer_34], properties {weights [1,1,1,1]}}
 {id=WavePlayer_12, properties {soundWaveAssetPtr=/Game/Sounds/Asteria/POI/Fortress/Spots/Spot_Fortress_Interior_Large_Stone_Debris_01}}
 {id=WavePlayer_13, properties {soundWaveAssetPtr=/Game/Sounds/Asteria/POI/Fortress/Spots/Spot_Fortress_Interior_Large_Stone_Debris_02}}
 {id=WavePlayer_14, properties {soundWaveAssetPtr=/Game/Sounds/Asteria/POI/Fortress/Spots/Spot_Fortress_Interior_Large_Stone_Debris_03}}
 {id=WavePlayer_15, properties {soundWaveAssetPtr=/Game/Sounds/Asteria/POI/Fortress/Spots/Spot_Fortress_Interior_Large_Stone_Debris_04}}
 {id=WavePlayer_16, properties {soundWaveAssetPtr=/Game/Sounds/Asteria/POI/Fortress/Spots/Spot_Fortress_Interior_Large_Stone_Debris_05}}
 {id=WavePlayer_17, properties {soundWaveAssetPtr=/Game/Sounds/Asteria/POI/Fortress/Spots/Spot_Fortress_Interior_Large_Stone_Debris_06}}
 {id=WavePlayer_24, properties {soundWaveAssetPtr=/Game/Sounds/Asteria/POI/Fortress/Spots/Spot_Fortress_Interior_Large_Stone_Debris_07}}
 {id=WavePlayer_26, properties {soundWaveAssetPtr=/Game/Sounds/Asteria/POI/Fortress/Spots/Spot_Fortress_Interior_Large_Stone_Debris_08}}
 {id=WavePlayer_27, properties {soundWaveAssetPtr=/Game/Sounds/Asteria/POI/Fortress/Spots/Spot_Fortress_Interior_Large_Stone_Debris_09}}
 {id=WavePlayer_28, properties {soundWaveAssetPtr=/Game/Sounds/Asteria/POI/Fortress/Spots/Spot_Fortress_Interior_Large_Stone_Debris_10}}
 {id=WavePlayer_29, properties {soundWaveAssetPtr=/Game/Sounds/Asteria/POI/Fortress/Spots/Spot_Fortress_Interior_Large_Stone_Debris_11}}
 {id=WavePlayer_30, properties {soundWaveAssetPtr=/Game/Sounds/Asteria/POI/Fortress/Spots/Spot_Fortress_Interior_Large_Stone_Debris_12}}
 {id=WavePlayer_31, properties {soundWaveAssetPtr=&Spot_Fortress_Interior_Large_Stone_Debris_01}}
 {id=WavePlayer_32, properties {soundWaveAssetPtr=&Spot_Fortress_Interior_Large_Stone_Debris_02}}
 {id=WavePlayer_33, properties {soundWaveAssetPtr=&Spot_Fortress_Interior_Large_Stone_Debris_03}}
 {id=WavePlayer_34, properties {soundWaveAssetPtr=&Spot_Fortress_Interior_Large_Stone_Debris_04}}
]
properties {
 attenuationSettings=/Game/Sounds/Attenuation/Fort_TileAmbient_OneShot_LessPositional_Log_Att
 maxDistance=6200
 pitchMultiplier=0.8
 soundClassObject=&Fort_Ambient_OneShots_Outdoors
 volumeMultiplier=0.376
}
