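# Looping spot ambience cue for wall wind at TheRaft FloatMax POI, reusing the Airborne City daytime stereo ambience wave.
# The looping wave feeds two nested, endlessly looping envelopers: a 30-second pitch curve inside a 20-second gentle volume dip.
# The differing loop lengths stagger pitch and volume movement.
# It plays at reduced volume as global ambient, using Desert Wasteland window-wind attenuation and a two-instance concurrency cap.
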SoundCue /Game/Sounds/Helios/POI/TheRaft_FloatMax/Spots/Spot_TheRaft_Floatmax_Wall_Wind_Loop_Cue
root=Enveloper_2
nodes [
 {
  id=Enveloper_1
  children [WavePlayer_0]
  properties {
   bLoop=true
   bLoopIndefinitely=true
   loopEnd=30
   pitchCurve {editorCurveData {keys [{interpMode=RCIM_Cubic, time=0}, <at 15.9468 0.9181 RCIM_Cubic>, {interpMode=RCIM_Cubic, time=30}]}}
   volumeCurve {editorCurveData {keys [<at 0 1>]}}
  }
 }
 {
  id=Enveloper_2
  children [Enveloper_1]
  properties {
   bLoop=true
   bLoopIndefinitely=true
   loopEnd=20
   pitchCurve {editorCurveData {keys [{time=0}]}}
   volumeCurve {editorCurveData {keys [
      {arriveTangent=-0.01, interpMode=RCIM_Cubic, leaveTangent=-0.01, time=0, value=1}
      <at 10 0.9 RCIM_Cubic>
      {arriveTangent=0.01, interpMode=RCIM_Cubic, leaveTangent=0.01, time=20, value=1}
     ]}}
  }
 }
 {id=WavePlayer_0, properties {bLooping=true, soundWaveAssetPtr=/Game/Sounds/Artemis/POI/AirborneCity/Ambience_AirborneCity_Day_Stereo}}
]
properties {
 attenuationSettings=/Game/Sounds/Helios/Biomes/DesertWasteland/Spots/Spot_Window_Wind_Sphere_Large_No_Focus_Att
 bOverrideConcurrency=true
 maxDistance=1450
 soundClassObject=/Game/Sounds/FortSoundClassesAndModes/Fort_Ambient_Global
 volumeMultiplier=0.351
 concurrencyOverrides {maxCount=2}
}
