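# Looping spot ambience cue of wind-rattled metal for TheRaft FloatMax POI, reusing a chainlink fence rattle wave from Hollow Base.
# Two nested looping envelopers drift pitch and swell volume over 40-second cycles.
# A quality node plays this chain in its first two slots and leaves the other two empty.
# Overridden attenuation adds listener focus, concurrency is capped, and the cue sends to the exterior ambient bus.
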SoundCue /Game/Sounds/Helios/POI/TheRaft_FloatMax/Spots/Spot_TheRaft_Floatmax_Wind_Metal_Aerial_Rattling_Cue
root=QualityLevel_0
nodes [
 {
  id=Enveloper_0
  children [WavePlayer_0]
  properties {
   bLoop=true
   bLoopIndefinitely=true
   loopEnd=40
   pitchCurve {editorCurveData {keys [
      {interpMode=RCIM_Cubic, time=0}
      <at 10 0.9 RCIM_Cubic>
      {arriveTangent=-0.0001, interpMode=RCIM_Cubic, leaveTangent=-0.0001, tangentMode=RCTM_User, time=20}
      <at 30 0.9 RCIM_Cubic>
      {interpMode=RCIM_Cubic, time=40}
     ]}}
   volumeCurve {editorCurveData {keys [<at 0 1>]}}
  }
 }
 {
  id=Enveloper_3
  children [Enveloper_0]
  properties {
   bLoop=true
   bLoopIndefinitely=true
   loopEnd=40
   pitchCurve {editorCurveData {keys [{time=0}]}}
   volumeCurve {editorCurveData {keys [
      {arriveTangent=-0.01, interpMode=RCIM_Cubic, time=0, value=1}
      <at 9.9314 0.3 RCIM_Cubic>
      {arriveTangent=-0.005, interpMode=RCIM_Cubic, leaveTangent=-0.005, time=20, value=1}
      <at 29.9314 0.2 RCIM_Cubic>
      <at 40 1>
     ]}}
  }
 }
 {id=QualityLevel_0, children [Enveloper_3,Enveloper_3,null,null]}
 {
  id=WavePlayer_0
  properties {bLooping=true, soundWaveAssetPtr=/Game/Sounds/Artemis/POI/Hollow_Base/Spots/Spot_Wind_Excited_Chainlink_Fence_Rattle_01}
 }
]
properties {
 attenuationSettings=/Game/Sounds/Artemis/POI/Hollow_Base/Spots/Spot_Interior_Window_Wind_Att
 bOverrideAttenuation=true
 bOverrideConcurrency=true
 maxDistance=1000
 soundClassObject=&Fort_Doors
 volumeMultiplier=0.15
 attenuationOverrides {
  bEnableListenerFocus=true
  falloffDistance=850
  focusAzimuth=25
  focusVolumeAttenuation=1.2
  nonFocusAzimuth=45
  nonFocusVolumeAttenuation=0.5
  stereoSpread=100
  attenuationShapeExtents {x=150}
 }
 busSends [{audioBus=null, minSendLevel=0.6, sendLevel=1, soundSourceBus=&AmbientAudio_Exterior_SourceBus}]
 concurrencyOverrides {maxCount=2, voiceStealReleaseTime=2}
}
